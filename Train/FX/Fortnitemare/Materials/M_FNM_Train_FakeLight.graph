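# Additive unlit Niagara particle material for the Fortnitemare train FX that fakes a point light on scene surfaces behind the particle.
# A sphere mask around the particle, scaled by particle size, is multiplied by scene-normal facing (constant on mobile) and particle alpha.
# Opacity fades as the camera nears, with depth fade on the default and deferred paths.
# Dynamic parameters drive radius, hardness and a two-sine flicker added to particle color.
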Material /Train/FX/Fortnitemare/Materials/M_FNM_Train_FakeLight
graph {
 properties {
  bEnableResponsiveAA=true
  bUsedWithNiagaraMeshParticles=true
  bUsedWithNiagaraSprites=true
  blendMode=BLEND_Additive
  pixelDepthOffsetMode=PDOM_Legacy
  shadingModel=MSM_Unlit
  usedShadingModels=Unlit
 }
 nodes [
  n1 = DynamicParameter() {desc="##1", parameterIndex=1, defaultValue=<rgba 1 0 0 0>, paramNames [LightRadiusScale, FlickerSpeed, FlickerScale, LightHardness]}
  n2 = Multiply(A=@n1.FlickerSpeed, B=Time()) {constA=1}
  n3 = Add(A=Sine(None=@n2) {period=0.5651}, B=Sine(None=@n2) {period=0.442})
  n4 = DynamicParameter() {desc="##1", parameterIndex=1, defaultValue=<rgba 1 0 0 0>, paramNames [LightRadiusScale, FlickerSpeed, FlickerScale, LightHardness]}
  n5 = ParticleColor()
  MaterialFunctionCall(Emissive=Add(A=Multiply(A=@n3, B=@n4.FlickerScale), B=@n5.RGB)) {materialFunction=&MF_FX_EmissiveAdjustments}
  n6 = SceneTexture() {sceneTextureId=PPI_WorldNormal}
  n7 = ParticlePositionWS()
  n8 = MaterialFunctionCall() {materialFunction=&WorldPositionBehindTranslucency}
  n9 = Subtract(A=@n7, B=@n8.Result)
  n10 = Saturate(None=DotProduct(A=ComponentMask(None=@n6.Color) {b=true, g=true, r=true}, B=Normalize(VectorInput=@n9)))
  n11 = ShadingPathSwitch(Default=@n10, Deferred=@n10, Mobile=Constant() {r=1}) {desc=Migrated from FeatureLevelSwitch}
  n12 = DynamicParameter() {desc="##1", parameterIndex=1, defaultValue=<rgba 1 0 0 0>, paramNames [LightRadiusScale, FlickerSpeed, FlickerScale, LightHardness]}
  n13 = ParticleSize()
  n14 = Power(
   Base=SphereMask(A=@n7, B=@n8.Result, Radius=Multiply(A=@n12.LightRadiusScale, B=@n13), Hardness=@n12.LightHardness)
   Exp=ScalarParameter() {defaultValue=6, parameterName=RadiusFalloffExponent}
  )
  n15 = MaterialFunctionCall(Float2=ParticleSize()) {materialFunction=&BreakOutFloat2Components}
  n16 = Divide(A=Distance(A=CameraPositionWS(), B=WorldPosition().XYZ), B=Min(A=@n15.R, B=@n15.R))
  n17 = Multiply(A=Multiply(A=@n5.A, B=Multiply(A=@n11, B=@n14)), B=Saturate(None=Subtract(A=@n16)))
  n18 = DepthFade(Opacity=@n17, FadeDistance=ScalarParameter() {defaultValue=50, parameterName=DepthFade})
  ShadingPathSwitch(Default=@n18, Deferred=@n18, Forward=@n17, Mobile=@n17)
 ]
}
