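# Translucent, unlit, two-sided flame trail material for Niagara ribbons and mesh particles in the Fortnitemare train FX.
# Panning caustic, blurred noise and trail textures, distorted by scrolling noise, form opacity shaped by head/tail masks and depth, view-angle and near-camera fades.
# Particle color feeds emissive adjustments and alpha.
# World position offset combines SplineThicken, a camera-space sine wave, and Niagara-driven bend and back offset through dynamic parameters.
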Material /Train/FX/Fortnitemare/Materials/M_FNM_Train_Flame_Trail
graph {
 properties {
  bUsedWithNiagaraMeshParticles=true
  bUsedWithNiagaraRibbons=true
  blendMode=BLEND_Translucent
  pixelDepthOffsetMode=PDOM_Legacy
  shadingModel=MSM_Unlit
  twoSided=true
  usedShadingModels=Unlit
 }
 nodes [
  n1 = DotProduct(
   A=MaterialFunctionCall() {materialFunction=&CameraDirectionVector}.Result
   B=Transform(None=Constant3Vector() {constant=<rgba 1 0 0 0>}) {transformSourceType=TRANSFORMSOURCE_Instance}
  )
  Abs(None=@n1)
  n2 = ParticleColor()
  MaterialFunctionCall(Emissive=@n2.RGB) {materialFunction=&MF_FX_EmissiveAdjustments}
  n3 = OneMinus(None=ComponentMask(None=NamedRerouteUsage()) {r=true})
  n4 = Saturate(None=Multiply(A=Power(Base=@n3) {constExponent=5}, B=ScalarParameter() {defaultValue=2, parameterName=PowerMask}) {constB=10})
  n5 = Constant()
  n6 = SmoothStep(Min=@n5, Max=Add(A=@n5, B=Constant() {r=0.333}), Value=@n3)
  n7 = DynamicParameter() {parameterIndex=1, defaultValue=<rgba 1 1 0 0>, paramNames [None, None, BackOffset, Use Trailing Fade]}
  n8 = NamedRerouteUsage()
  n9 = AppendVector(A=NamedRerouteUsage(), B=Constant())
  n10 = NamedRerouteUsage()
  n11 = Multiply(
   A=Time()
   B=Multiply(
    A=ScalarParameter() {defaultValue=1, parameterName=Speed Direction}
    B=DynamicParameter() {parameterIndex=2, defaultValue=<rgba 1 1 1 0>, paramNames [Speed Dir Mult, None, None, None]}.Speed Dir Mult
   )
  )
  n12 = ComponentMask(None=TextureSample(UVs=Panner(Coordinate=@n10, Time=@n11) {speedX=-0.1, speedY=0.2}) {samplerType=SAMPLERTYPE_Masks, texture=/Game/Weapons/WeaponSkins/Wraps/Textures/Mash/Mash_Noise}.RGB) {g=true, r=true}
  n13 = Add(A=Add(A=Add(A=@n8, B=@n9) {constB=0}, B=Multiply(A=@n12) {constB=0.1}), B=Constant2Vector() {g=0.7})
  n14 = Multiply(
   A=Time()
   B=Multiply(
    A=ScalarParameter() {defaultValue=1, parameterName=Speed Direction}
    B=DynamicParameter() {parameterIndex=2, defaultValue=<rgba 1 1 1 0>, paramNames [Speed Dir Mult, None, None, None]}.Speed Dir Mult
   )
  )
  n15 = TextureSample(UVs=Panner(Coordinate=@n13, Time=@n14) {speedX=-1}) {samplerType=SAMPLERTYPE_LinearGrayscale, texture=/Game/Global/Textures/Default/Caustics/T_Nosie_Caustics_Grayscale}
  n16 = TextureSample(UVs=Panner(Coordinate=@n13, Time=@n14) {speedX=-0.5}) {samplerType=SAMPLERTYPE_LinearGrayscale, texture=&T_Nosie_Caustics_Grayscale}
  n17 = NamedRerouteUsage()
  n18 = Panner(Coordinate=Add(A=Add(A=@n17, B=@n9) {constB=0}, B=Constant2Vector() {g=0.6}), Time=@n14) {speedX=-1}
  n19 = Power(Base=LinearInterpolate(Alpha=TextureSample(UVs=@n18) {samplerType=SAMPLERTYPE_Grayscale, texture=&LowResBlurredNoise}.R)) {constExponent=1}
  n20 = Power(Base=Multiply(A=LinearInterpolate(A=@n15.R, B=@n16.R), B=Saturate(None=@n19))) {constExponent=1}
  n21 = NamedRerouteUsage()
  n22 = Constant()
  n23 = Add(A=Add(A=AppendVector(A=@n21, B=@n22), B=NamedRerouteUsage()), B=Multiply(A=@n12) {constB=0.12})
  n24 = Multiply(
   A=Time()
   B=Multiply(
    A=ScalarParameter() {defaultValue=1, parameterName=Speed Direction}
    B=DynamicParameter() {parameterIndex=2, defaultValue=<rgba 1 1 1 0>, paramNames [Speed Dir Mult, None, None, None]}.Speed Dir Mult
   )
  )
  n25 = Multiply(A=TextureSample(UVs=Panner(Coordinate=@n23, Time=@n24) {speedX=-1}) {texture=/Game/Gadgets/Assets/VinderTech_GliderChute/Glider_DarkStance/FX/T_Glider_DarkStance_Trail}.G) {constB=2}
  n26 = Multiply(A=Saturate(None=Multiply(A=@n25, B=@n20)), B=LinearInterpolate(A=@n4, B=Saturate(None=@n6), Alpha=@n7.Use Trailing Fade) {constA=1})
  n27 = NamedRerouteUsage()
  n28 = Multiply(A=Subtract(A=ComponentMask(None=@n27) {r=true}, B=ScalarParameter() {parameterName=Lead Mask Offset}) {constB=0}) {constB=5}
  n29 = Multiply(A=@n2.A, B=Multiply(A=Multiply(A=@n26, B=@n4), B=Saturate(None=@n28)))
  n30 = DepthFade(Opacity=@n29, FadeDistance=ScalarParameter() {defaultValue=5, parameterName=DepthFade})
  n31 = ShadingPathSwitch(Default=@n30, Deferred=@n30, Forward=@n29, Mobile=@n29)
  n32 = NamedRerouteUsage()
  n33 = ScalarParameter() {defaultValue=0.25, parameterName=AngleFade}
  Multiply(
   A=Multiply(
    A=Multiply(A=@n31, B=@n32)
    B=MaterialFunctionCall("FadeSensitivity (optional)"=@n33) {materialFunction=&FadeBasedOnViewAngle}.Result
   )
   B=NamedRerouteUsage()
  )
  NamedRerouteDeclaration(None=Length(None=VertexColor()) {is=class MaterialExpressionLength}) {name=Offset, nodeColor=<rgba 1 0.7031 0 1>}
  n34 = TextureCoordinate()
  n35 = MaterialFunctionCall(Float2=@n34) {materialFunction=&BreakOutFloat2Components}
  n36 = OneMinus(None=@n35.R)
  NamedRerouteDeclaration(None=StaticSwitchParameter(
    True=MaterialFunctionCall(X=@n36, Y=@n35.G) {materialFunction=&MakeFloat2}.Result
    False=@n34
   ) {defaultValue=true, parameterName=InverseUV}) {name=UV's, nodeColor=<rgba 0.7812 1 0 1>}
  n37 = Power(
   Base=MaterialFunctionCall(Distance=ScalarParameter() {defaultValue=100, parameterName=NearCamFadeDistance}) {materialFunction=&Near_Camera_Fade}.Result
   Exp=ScalarParameter() {defaultValue=2, parameterName=NearCamFadePower}
  )
  n38 = Constant() {r=1}
  NamedRerouteDeclaration(None=QualitySwitch(Default=@n37, Low=@n38, High=@n37, Medium=@n38, Epic=@n37)) {name=NearCameraFade, nodeColor=<rgba 1 0 0.7734 1>}
  n39 = Constant() {r=1}
  n40 = OneMinus(None=Power(Base=Saturate(None=@n1), Exp=Constant() {r=6}))
  n41 = ScalarParameter() {parameterName=UseFrontBackFade?}
  NamedRerouteDeclaration(None=Saturate(None=LinearInterpolate(A=@n39, B=@n40, Alpha=@n41))) {name=FrontBackFade, nodeColor=<rgba 1 0.2578 0 1>}
  n42 = DynamicParameter() {desc=Parameters set by Niagara, parameterIndex=3, paramNames [X,Y,Z,WPOIntensity]}
  n43 = Subtract(
   A=MaterialFunctionCall(X=@n42.X, Y=@n42.X, Z=@n42.X) {materialFunction=&MakeFloat3}.Result
   B=ParticlePositionWS()
  )
  n44 = ScalarParameter() {defaultValue=-35, group=WPO, parameterName=Clamp_min}
  n45 = ScalarParameter() {defaultValue=35, group=WPO, parameterName=Clamp_max}
  n46 = Power(Base=ComponentMask(None=NamedRerouteUsage()) {r=true}, Exp=ScalarParameter() {defaultValue=3, group=WPO, parameterName=WPOBendPower})
  n47 = LinearInterpolate(
   B=LinearInterpolate(B=Clamp(None=@n43, Min=@n44, Max=@n45) {maxDefault=1000, minDefault=-1000}, Alpha=Multiply(A=@n46, B=@n42.WPOIntensity))
   Alpha=ScalarParameter() {defaultValue=1, parameterName=Use interpolate}
  )
  n48 = DynamicParameter() {defaultValue=<rgba 1 1 0 0>, paramNames [None, None, V offset, U offset]}
  n49 = NamedRerouteUsage()
  n50 = AppendVector(A=NamedRerouteUsage(), B=Constant() {r=1})
  n51 = Constant() {r=1}
  n52 = Constant() {r=1}
  n53 = ComponentMask(None=Multiply(A=Add(A=@n49, B=@n50), B=AppendVector(A=@n51, B=@n52))) {r=true}
  n54 = Multiply(A=Time()) {constB=-0.7}
  n55 = Multiply(
   A=Sine(None=Add(A=@n53, B=@n54))
   B=Transform(None=VectorParameter() {parameterName=WPO Direction, defaultValue=<rgba 0 1 0 1>}.RGB) {transformSourceType=TRANSFORMSOURCE_Camera}
  )
  n56 = ScalarParameter() {defaultValue=5, parameterName=WPO Intensity}
  n57 = Multiply(A=ComponentMask(None=NamedRerouteUsage()) {r=true}) {constB=7}
  n58 = Add(
   A=MaterialFunctionCall(
    WidthBase=@n48.R
    WidthTip=@n48.R
    Expand U or V UV Channel=StaticBoolParameter() {parameterName="Expand U or V UV Channel (SB)"}
    AdditionalNormal=VectorParameter() {defaultValue=<rgba 0 0 1 1>}.RGB
   ) {materialFunction=&SplineThicken}.WorldPositionOffset
   B=Multiply(A=Multiply(A=@n55, B=@n56), B=Saturate(None=@n57))
  )
  n59 = Power(Base=ComponentMask(None=NamedRerouteUsage()) {r=true}, Exp=ScalarParameter() {defaultValue=1.5, group=WPO, parameterName=WPOBendPowerFE})
  n60 = Constant()
  n61 = DynamicParameter() {parameterIndex=1, defaultValue=<rgba 1 1 0 0>, paramNames [None, None, BackOffset, Use Trailing Fade]}
  n62 = Transform(None=MaterialFunctionCall(X=@n60, Y=@n61.BackOffset, Z=@n60) {materialFunction=&MakeFloat3}.Result) {transformSourceType=TRANSFORMSOURCE_World, transformType=TRANSFORM_Local}
  Reroute(None=Add(A=Add(A=@n58, B=@n47) {constB=0}, B=Multiply(A=@n59, B=@n62)))
 ]
}
