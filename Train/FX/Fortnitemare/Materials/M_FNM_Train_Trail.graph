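# Translucent unlit trail material for Niagara mesh particles in the Fortnitemares train effects, tinted by particle color.
# Panned smoke, streak and caustic textures, distorted by noise, build opacity shaped by tail, appear and lead-edge masks.
# World position offset combines spline thickening, sine waving, back offset and a Niagara-driven bend, steered by dynamic parameters.
# Depth, view-angle and near-camera fades soften it, with cheaper paths on low quality and mobile.
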Material /Train/FX/Fortnitemare/Materials/M_FNM_Train_Trail
graph {
 properties {
  bUsedWithNiagaraMeshParticles=true
  blendMode=BLEND_Translucent
  pixelDepthOffsetMode=PDOM_Legacy
  shadingModel=MSM_Unlit
  usedShadingModels=Unlit
 }
 nodes [
  n1 = DotProduct(
   A=MaterialFunctionCall() {materialFunction=&CameraDirectionVector}.Result
   B=Transform(None=Constant3Vector() {constant=<rgba 1 0 0 0>}) {transformSourceType=TRANSFORMSOURCE_Instance}
  )
  Abs(None=@n1)
  Constant() {r=-0.5}
  Constant() {r=1.5}
  n2 = PixelNormalWS()
  n3 = VertexNormalWS()
  n4 = QualitySwitch(Default=@n2, Low=@n3, High=@n2, Medium=@n3, Epic=@n2)
  n5 = Abs(None=DotProduct(A=CameraVectorWS(), B=ShadingPathSwitch(Default=@n4, Deferred=@n4, Forward=@n3, Mobile=@n3)))
  n6 = VectorParameter() {
   group=Base
   parameterName=Mesh Fresnel Fade
   sortPriority=1
   channelNames {a="", b=Power, g=Invert, r=Use Mesh Fresnel Fade}
   defaultValue=<rgba 1 0 0.1 0>
  }
  n7 = Power(Base=LinearInterpolate(A=@n5, B=OneMinus(None=@n5), Alpha=@n6.Invert), Exp=@n6.Power)
  n8 = OneMinus(None=ComponentMask(None=TextureCoordinate()) {r=true})
  n9 = Multiply(A=Power(Base=Add(A=@n8) {constB=0.6}) {constExponent=10}) {constB=5}
  LinearInterpolate(B=Multiply(A=Saturate(None=@n7), B=Saturate(None=@n9)), Alpha=@n6.Use Mesh Fresnel Fade) {constA=1}
  n10 = ParticleColor()
  MaterialFunctionCall(Emissive=@n10.RGB) {materialFunction=&MF_FX_EmissiveAdjustments}
  n11 = TextureCoordinate()
  MaterialFunctionCall(XY=@n11) {materialFunction=&Swizzle}
  n12 = DynamicParameter() {parameterIndex=1, defaultValue=<rgba 1 1 0 0>, paramNames [Use Forward Tail mask, Appear Mask, BackOffset, Use Trailing Fade]}
  n13 = NamedRerouteUsage()
  n14 = AppendVector(A=DynamicParameter() {defaultValue=<rgba 200 50 1 0>, paramNames ["", "", U_Scale, U offset]}.U offset, B=Constant())
  n15 = Constant2Vector() {r=-0.5}
  n16 = NamedRerouteUsage()
  n17 = TextureSample(UVs=Panner(Coordinate=Add(A=@n13, B=@n14), Speed=Multiply(A=@n15, B=@n16)) {speedX=-0.5}) {samplerType=SAMPLERTYPE_Masks, texture=/Game/Weapons/WeaponSkins/Wraps/Textures/Mash/Mash_Noise}
  n18 = DynamicParameter() {parameterIndex=1, defaultValue=<rgba 1 1 0 0>, paramNames [Use Forward Tail mask, Appear Mask, BackOffset, Use Trailing Fade]}
  n19 = ComponentMask(None=NamedRerouteUsage()) {r=true}
  n20 = Saturate(None=LinearInterpolate(
    A=Constant() {r=-1}
    B=LinearInterpolate(Alpha=@n18.Use Forward Tail mask) {constA=0.5, constB=1.5}
    Alpha=Power(Base=@n19) {constExponent=1.3}
   ) {constA=0.1, constB=2})
  n21 = Saturate(None=LinearInterpolate(A=LinearInterpolate(Alpha=@n12.Use Forward Tail mask) {constB=-0.2}, B=Constant() {r=1.2}, Alpha=Power(Base=@n17.R, Exp=@n20)) {constA=-2, constB=2})
  n22 = Constant2Vector() {r=-1}
  n23 = NamedRerouteUsage()
  n24 = TextureSample(UVs=Panner(Coordinate=NamedRerouteUsage(), Speed=Multiply(A=@n22, B=@n23)) {speedX=-1}) {samplerType=SAMPLERTYPE_Grayscale, texture=&T_Smoke_Trail_Soft}
  n25 = Panner(Coordinate=NamedRerouteUsage(), Speed=Multiply(A=Constant2Vector() {g=0.2, r=-0.1}, B=NamedRerouteUsage())) {speedX=-0.1, speedY=0.2}
  n26 = Multiply(A=ComponentMask(None=TextureSample(UVs=@n25) {samplerType=SAMPLERTYPE_Masks, texture=&Mash_Noise}.RGB) {g=true, r=true}) {constB=0.12}
  n27 = Add(A=NamedRerouteUsage(), B=@n26)
  n28 = AppendVector(A=DynamicParameter() {defaultValue=<rgba 200 50 1 0>, paramNames ["", "", U_Scale, U offset]}.U offset, B=Constant())
  n29 = Constant2Vector() {r=-1}
  n30 = NamedRerouteUsage()
  n31 = TextureSample(UVs=Panner(Coordinate=Add(A=@n27, B=@n28), Speed=Multiply(A=@n29, B=@n30)) {speedX=-1}) {texture=/Game/Gadgets/Assets/VinderTech_GliderChute/Glider_DarkStance/FX/T_Glider_DarkStance_Trail}
  n32 = Power(Base=Add(A=Multiply(A=@n24.R), B=Multiply(A=@n31.G) {constB=2}) {constB=0})
  n33 = AppendVector(A=DynamicParameter() {defaultValue=<rgba 200 50 1 0>, paramNames ["", "", U_Scale, U offset]}.U_Scale, B=Constant() {r=1})
  n34 = NamedRerouteUsage()
  n35 = Add(
   A=Add(A=Multiply(A=@n33, B=@n34), B=@n26)
   B=AppendVector(A=DynamicParameter() {defaultValue=<rgba 200 50 1 0>, paramNames ["", "", U_Scale, U offset]}.U offset, B=Constant())
  )
  n36 = Multiply(A=Constant2Vector() {r=-1}, B=NamedRerouteUsage())
  n37 = Constant2Vector() {r=-0.5}
  n38 = NamedRerouteUsage()
  n39 = TextureSample(UVs=Panner(Coordinate=NamedRerouteUsage(), Speed=Multiply(A=@n37, B=@n38)) {speedX=-0.5}) {
   samplerType=SAMPLERTYPE_LinearColor
   texture=/Engine/Functions/Engine_MaterialFunctions02/ExampleContent/Textures/OffsetNoiseDistanceFields_ForNormals
  }
  n40 = Multiply(
   A=TextureSample(UVs=Panner(Coordinate=@n35, Speed=@n36) {speedX=-1}) {texture=&T_Glider_DarkStance_Trail}.R
   B=LinearInterpolate(Alpha=@n39.R) {constA=0.4}
  )
  n41 = Saturate(None=Add(A=Add(A=@n32, B=@n40)) {constB=0})
  n42 = Power(Base=ComponentMask(None=NamedRerouteUsage()) {r=true}) {constExponent=1.5}
  n43 = Multiply(A=@n41, B=Power(Base=LinearInterpolate(Alpha=@n39.R) {constA=0.5}, Exp=LinearInterpolate(Alpha=@n42)) {constExponent=1})
  n44 = LinearInterpolate(Alpha=ComponentMask(None=NamedRerouteUsage()) {r=true}) {constA=0.8}
  n45 = Multiply(
   A=AppendVector(A=DynamicParameter() {defaultValue=<rgba 200 50 1 0>, paramNames ["", "", U_Scale, U offset]}.U_Scale, B=Constant() {r=1})
   B=NamedRerouteUsage()
  )
  n46 = Multiply(A=Constant2Vector() {r=-1}, B=NamedRerouteUsage())
  n47 = TextureSample(UVs=Add(A=Panner(Coordinate=@n45, Speed=@n46) {speedX=-1}, B=Multiply(A=@n26))) {samplerType=SAMPLERTYPE_LinearGrayscale, texture=/Game/Global/Textures/Default/Caustics/T_Nosie_Caustics_Grayscale}
  n48 = ComponentMask(None=NamedRerouteUsage()) {r=true}
  n49 = Multiply(A=Multiply(A=Power(Base=@n47.R) {constExponent=1}) {constB=1.5}, B=Power(Base=OneMinus(None=@n48)) {constExponent=1})
  n50 = Saturate(None=Add(A=Multiply(A=@n43, B=@n44), B=Multiply(A=@n49, B=@n41)))
  n51 = OneMinus(None=ComponentMask(None=NamedRerouteUsage()) {r=true})
  n52 = Saturate(None=Multiply(A=@n51, B=Constant() {r=10}) {constB=10})
  n53 = Constant()
  n54 = Constant() {r=0.6}
  n55 = Saturate(None=SmoothStep(Min=@n53, Max=Add(A=@n53, B=@n54), Value=@n51))
  n56 = DynamicParameter() {parameterIndex=1, defaultValue=<rgba 1 1 0 0>, paramNames [Use Forward Tail mask, Appear Mask, BackOffset, Use Trailing Fade]}
  n57 = Constant()
  n58 = Add(A=@n57, B=ScalarParameter() {defaultValue=0.25, parameterName=Base_Fade})
  n59 = ComponentMask(None=NamedRerouteUsage()) {r=true}
  n60 = Multiply(
   A=Multiply(A=Multiply(A=@n50, B=@n21), B=LinearInterpolate(A=@n52, B=@n55, Alpha=@n56.Use Trailing Fade) {constA=1})
   B=LinearInterpolate(B=SmoothStep(Min=@n57, Max=@n58, Value=@n59), Alpha=ScalarParameter() {defaultValue=1, parameterName=LeadEdge Opacity}) {constA=1}
  )
  n61 = ComponentMask(None=NamedRerouteUsage()) {r=true}
  n62 = DynamicParameter() {parameterIndex=1, defaultValue=<rgba 1 1 0 0>, paramNames [Use Forward Tail mask, Appear Mask, BackOffset, Use Trailing Fade]}
  n63 = Multiply(A=Add(A=OneMinus(None=@n61), B=LinearInterpolate(Alpha=@n62.Appear Mask) {constA=-1, constB=0.1})) {constB=10}
  n64 = Multiply(A=Multiply(A=Multiply(A=@n10.A, B=@n60), B=Saturate(None=@n63)))
  n65 = DepthFade(Opacity=@n64, FadeDistance=ScalarParameter() {defaultValue=10, parameterName=DepthFade})
  n66 = Multiply(A=ShadingPathSwitch(Default=@n65, Deferred=@n65, Forward=@n64, Mobile=@n64), B=NamedRerouteUsage())
  n67 = NamedRerouteUsage()
  n68 = Constant() {r=0.25}
  Multiply(
   A=Multiply(
    A=Multiply(A=@n66, B=@n67)
    B=MaterialFunctionCall("FadeSensitivity (optional)"=@n68) {materialFunction=&FadeBasedOnViewAngle}.Result
   )
   B=NamedRerouteUsage()
  )
  n69 = DynamicParameter() {defaultValue=<rgba 200 50 1 0>, paramNames ["", "", U_Scale, U offset]}
  n70 = ScalarParameter() {defaultValue=5, parameterName=Base}
  NamedRerouteDeclaration(None=MaterialFunctionCall(WidthBase=Add(A=@n69.R, B=@n70), WidthTip=@n69.G, Expand U or V UV Channel=StaticBool()) {materialFunction=&SplineThicken}.WorldPositionOffset) {name=SplineThicken, nodeColor=<rgba 0 1 1 1>}
  n71 = MaterialFunctionCall(Float2=@n11) {materialFunction=&BreakOutFloat2Components}
  NamedRerouteDeclaration(None=MaterialFunctionCall(X=OneMinus(None=@n71.R), Y=@n71.G) {materialFunction=&MakeFloat2}.Result) {name=UV's, nodeColor=<rgba 0.7812 1 0 1>}
  n72 = ComponentMask(None=NamedRerouteUsage()) {r=true}
  n73 = Constant() {r=1.5}
  n74 = DynamicParameter() {parameterIndex=1, defaultValue=<rgba 1 1 0 0>, paramNames [Use Forward Tail mask, Appear Mask, BackOffset, Use Trailing Fade]}
  n75 = MaterialFunctionCall(X=Constant(), Y=@n74.BackOffset, Z=ScalarParameter() {defaultValue=-50, parameterName=Back Offset}) {materialFunction=&MakeFloat3}
  n76 = Add(
   A=NamedRerouteUsage()
   B=Multiply(A=Power(Base=@n72, Exp=@n73), B=Transform(None=@n75.Result) {transformSourceType=TRANSFORMSOURCE_World, transformType=TRANSFORM_Local})
  )
  n77 = NamedRerouteUsage()
  NamedRerouteDeclaration(None=Add(A=Add(A=@n76, B=@n77) {constB=0}, B=NamedRerouteUsage())) {name=WPO, nodeColor=<rgba 0.17 0.0489 0 1>}
  n78 = NamedRerouteUsage()
  n79 = DynamicParameter() {defaultValue=<rgba 200 50 1 0>, paramNames ["", "", U_Scale, U offset]}
  n80 = AppendVector(A=@n79.U offset, B=Constant())
  n81 = Constant() {r=1}
  n82 = ComponentMask(None=Multiply(A=Add(A=@n78, B=@n80), B=AppendVector(A=@n79.U_Scale, B=@n81))) {r=true}
  n83 = Multiply(A=Time()) {constB=-0.7}
  n84 = Multiply(
   A=Sine(None=Add(A=@n82, B=@n83))
   B=Transform(None=VectorParameter() {parameterName=WPO Direction, defaultValue=<rgba 0 1 0 1>}.RGB) {transformSourceType=TRANSFORMSOURCE_Camera}
  )
  n85 = ScalarParameter() {defaultValue=5, parameterName=WPO Intensity}
  n86 = Multiply(A=ComponentMask(None=NamedRerouteUsage()) {r=true}) {constB=7}
  NamedRerouteDeclaration(None=Multiply(A=Multiply(A=@n84, B=@n85), B=Saturate(None=@n86))) {name=WPO Distortion, nodeColor=<rgba 0.1953 1 0 1>}
  n87 = DynamicParameter() {desc=Parameters set by Niagara, parameterIndex=3, paramNames [X,Y,Z,WPOIntensity]}
  n88 = MaterialFunctionCall(X=@n87.X, Y=@n87.X, Z=@n87.X) {materialFunction=&MakeFloat3}
  n89 = ParticlePositionWS()
  n90 = Clamp(
   None=Multiply(B=Subtract(A=@n88.Result, B=@n89)) {constA=1}
   Min=ScalarParameter() {defaultValue=-35, group=WPO, parameterName=Clamp_min_X}
   Max=ScalarParameter() {defaultValue=35, group=WPO, parameterName=Clamp_max}
  ) {maxDefault=1000, minDefault=-1000}
  n91 = NamedRerouteUsage()
  n92 = Multiply(
   A=Power(Base=ComponentMask(None=@n91) {r=true}, Exp=ScalarParameter() {defaultValue=3, group=WPO, parameterName=WPOBendPower})
   B=@n87.WPOIntensity
  )
  NamedRerouteDeclaration(None=LinearInterpolate(B=LinearInterpolate(B=@n90, Alpha=@n92), Alpha=ScalarParameter() {defaultValue=1, parameterName=Use interpolate})) {name=WPO Interpolate, nodeColor=<rgba 1 0 0.0703 1>}
  n93 = Power(
   Base=MaterialFunctionCall(Distance=DynamicParameter() {parameterIndex=2, defaultValue=<rgba 1 100 1 0>, paramNames [Speed Dir Mult, NearCameraFade, "", ""]}.NearCameraFade) {materialFunction=&Near_Camera_Fade}.Result
   Exp=ScalarParameter() {defaultValue=2, parameterName=NearCamFadePower}
  )
  n94 = Constant() {r=1}
  NamedRerouteDeclaration(None=QualitySwitch(Default=@n93, Low=@n94, High=@n93, Medium=@n94, Epic=@n93)) {name=NearCameraFade, nodeColor=<rgba 1 0 0.7734 1>}
  NamedRerouteDeclaration(None=ScalarParameter() {defaultValue=1, parameterName=PanSpeedMultiplier}) {name=PanSpeedMultiplier, nodeColor=<rgba 0 1 0.7656 1>}
  n95 = Constant() {r=0.436}
  n96 = Add(A=@n95, B=Constant() {r=0.2})
  n97 = MaterialFunctionCall(Float2=TextureCoordinate()) {materialFunction=&BreakOutFloat2Components}
  n98 = OneMinus(None=Abs(None=ConstantBiasScale(None=@n97.G) {bias=-0.5, scale=2}))
  n99 = Constant() {r=0.8}
  n100 = Add(A=@n99, B=Constant() {r=0.2})
  n101 = MaterialFunctionCall(Float2=TextureCoordinate()) {materialFunction=&BreakOutFloat2Components}
  NamedRerouteDeclaration(None=LinearInterpolate(A=Constant() {r=1}, B=SmoothStep(Min=@n95, Max=@n96, Value=@n98), Alpha=SmoothStep(Min=@n99, Max=@n100, Value=@n101.R)) {constA=1}) {name=Bottom Mask, nodeColor=<rgba 1 0.2578 0 1>}
  n102 = Constant() {r=1}
  n103 = OneMinus(None=Power(Base=Saturate(None=@n1), Exp=Constant() {r=2}))
  n104 = ScalarParameter() {parameterName=UseFrontBackFade?}
  NamedRerouteDeclaration(None=Saturate(None=LinearInterpolate(A=@n102, B=@n103, Alpha=@n104))) {name=FrontBackFade, nodeColor=<rgba 1 0.2578 0 1>}
  NamedRerouteUsage()
  OneMinus(None=@n91)
  TextureCoordinate()
  TextureCoordinate()
 ]
}
