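# Translucent unlit Niagara sprite material for small flames in the Fortnitemare FX set.
# A panning Perlin texture, offset per particle, distorts the flame spritesheet UVs, with a sine mask pinning the left and right edges.
# Flame green minus inverted particle alpha is sharpened and indexes a Ghost curve-atlas row, tinted by particle color for emissive.
# Flame green is also multiplied by particle alpha, separately from the color chain.
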Material /Train/FX/Fortnitemare/Materials/M_SmallFlames
graph {
 properties {bUsedWithNiagaraSprites=true, blendMode=BLEND_Translucent, pixelDepthOffsetMode=PDOM_Legacy, shadingModel=MSM_Unlit, usedShadingModels=Unlit}
 nodes [
  n1 = TextureCoordinate()
  n2 = MaterialFunctionCall(Float2=Sine(None=Multiply(A=@n1) {constB=0.5})) {materialFunction=&BreakOutFloat2Components}
  n3 = Saturate(None=Multiply(A=Multiply(A=@n2.R, B=@n2.R)) {constB=4})
  n4 = Add(A=@n1, B=ParticleRandom())
  n5 = TextureSampleParameter2D(UVs=Panner(Coordinate=Multiply(A=@n4) {constB=1.2}) {speedX=-0.06, speedY=0.5}) {
   parameterName=TexDistort
   samplerType=SAMPLERTYPE_Masks
   texture=/Game/Characters/Player/Male/Large/Bodies/M_LRG_Clash_V/Textures/Simple_Perlin_M
  }
  n6 = Multiply(A=@n5.R) {constB=0.1}
  n7 = TextureSampleParameter2D(UVs=Add(A=@n1, B=Multiply(A=@n3, B=@n6))) {
   parameterName=Tex1
   samplerType=SAMPLERTYPE_LinearGrayscale
   texture=/Game/Weapons/FORT_Melee/Pickaxe_RippedHarvester/FX/T_RippedHarvester_Flame_Spritesheet
  }
  n8 = ParticleColor()
  n9 = Subtract(A=@n7.G, B=OneMinus(None=@n8.A))
  n10 = Add(A=Multiply(A=Power(Base=@n9) {constExponent=20})) {constB=0.7}
  n11 = CurveAtlasRowParameter(CurveTime=OneMinus(None=Saturate(None=@n10))) {
   atlas=&AthenaCurveAtlas01
   curve=/Game/Athena/LinearColorCurves/Curve_Ghost_V
   defaultValue=51
   parameterName=ColorCurve
  }
  MaterialFunctionCall(Emissive=Multiply(A=@n11, B=@n8.RGB)) {materialFunction=&MF_FX_EmissiveAdjustments}
  Multiply(A=@n7.G, B=@n8.A)
 ]
}
