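# Niagara system for Fortnitemares train wheel effects: teal sparks, skid meshes with fake-light glow, and dark smoke with energy trails.
# User.isFrontPart switches spawn counts and wheel offset arrays between front part and car layouts, using particle partitioning to place effects.
# A normalized system velocity value scales spark spawn probability, mesh height and glow alpha.
# GPU sparks use curl noise, drag, speed stretching and a kill volume; a debug gnomon is disabled.
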NiagaraSystem /Train/FX/Fortnitemare/NS_FNM_Train_Wheels_Car
dependencies {
 usedDataInterfaces [
  &NiagaraDataInterfaceCurve
  &NiagaraDataInterfaceArrayFloat3
  &NiagaraDataInterfaceArrayFloat
  &NiagaraDataInterfaceColorCurve
 ]
 usedDynamicInputs [&ClampFloat, &RemapRange, &VectorLength]
 usedModules [
  &SystemState
  &self:SystemSpawnScript.SystemScriptSource.SystemScriptGraph.NiagaraNodeAssignment_3.SetVariables_#1
  &self:SystemSpawnScript.SystemScriptSource.SystemScriptGraph.NiagaraNodeAssignment_2.SetVariables_#2
  &self:SystemSpawnScript.SystemScriptSource.SystemScriptGraph.NiagaraNodeAssignment_1.SetVariables_#3
 ]
}
emitters {
 FN_DebugGnomon {
  bEnabled=false
  emitterUpdateScript [
   &EmitterState(
    Life Cycle Mode=::NewEnumerator1
    Loop Behavior=::NewEnumerator1
    Loop Duration Mode=::NewEnumerator1
    Scalability Mode=::NewEnumerator1
    UseLoopDelay=false
   )
   &SpawnBurst_Instantaneous(Use Spawn Probability=false, UseLoopCountLimit=false)
  ]
  particleSpawnScript [&InitializeParticle(
    Color Mode=::NewEnumerator0
    Lifetime=5
    Mesh Scale Mode=::NewEnumerator3
    Mesh Uniform Scale=20
    UsePositionOffset=false
    Write Material Random=true
   )]
  particleUpdateScript [
   &ParticleState(Kill Particles When Lifetime Has Elapsed=false)
   /SetVariables(Particles.Position=/Niagara/DynamicInputs/Helpers/SimulationPosition()) {isSetParameters=true}
  ]
  rendererClasses [&NiagaraMeshRendererProperties]
  renderers [{rendererClass=&NiagaraMeshRendererProperties, rendererIndex=0}]
 }
 SkidSharp {
  emitterUpdateScript [
   &EmitterState(Life Cycle Mode=::NewEnumerator1, Loop Behavior=::NewEnumerator1, Loop Duration Mode=::NewEnumerator1, UseLoopDelay=false)
   &SpawnRate(
    SpawnRate=&MakeCustomFloatFromBool(Bool=$User.isFrontPart, True Float=20, False Float=40)
    Use Spawn Probability=false
   )
   &SpawnBurst_Instantaneous(
    Spawn Count=&MakeCustomIntFromBool(Bool=$User.isFrontPart, True INT=2, False INT=4)
    Spawn Group=1
    Use Spawn Probability=false
    UseLoopCountLimit=false
   )
  ]
  particleSpawnScript [
   /SetVariables(Particles.SpawnGroup=$Emitter.SpawnGroup) {isSetParameters=true}
   /SetVariables(Particles.VisibilityTag=$Particles.SpawnGroup) {isSetParameters=true}
   &PartitionParticles(Number Of Partitions=&MakeCustomIntFromBool(Bool=$User.isFrontPart, True INT=2, False INT=4))
   &InitializeParticle(
    Color=<rgba 10 10 10 1>
    Lifetime=&MaskFloatBySpawnGroup(Masked Float=&RandomRangeFloat(Minimum=0.1, Maximum=0.15), Passthrough Float=86400)
    Mesh Scale=&RandomRangeVector(Minimum=<0.6 0.25 0.1>, Maximum=<1.5 0.5 0.25>)
    Mesh Scale Mode=::NewEnumerator4
    Position Offset=<5 0 0>
    Sprite Size Mode=::NewEnumerator3
    Uniform Sprite Size=220
    UsePositionOffset=true
    Write Material Random=true
   ) {module=InitializeParticle001}
   &DynamicMaterialParameters(
    Index 0 Param 3=&RandomRangeFloat(Minimum=1, Maximum=1.5, Evaluation Type=::NewEnumerator1)
    Module.Param0WriteEnabled=false
    Module.Param1WriteEnabled=false
    Module.Param2WriteEnabled=true
    Module.Param3WriteEnabled=false
   ) {module=DynamicMaterialParameters001}
   &CameraOffset()
   &AddVelocity(Velocity=&SelectVectorFromArray(
     Direct Array Index=&MakeCustomIntFromBool(
      Bool=$User.isFrontPart
      True INT=&ModuloInt(A=$Particles.PartitionParticles.Partition, B=2)
      False INT=$Particles.PartitionParticles.Partition
     )
     Vector Selection Array=DataInterface() {propertyValues {FloatData [<-1 0.25 0>, <-1 -0.25 0>, <-1 0.25 0>, <-1 -0.25 0>]}}
    ))
   &SystemLocation(
    MaskSpawnGroup=false
    Offset=&MakeCustomVectorFromBool(
     Bool=$User.isFrontPart
     True Vector=&SelectVectorFromArray(Direct Array Index=$Particles.PartitionParticles.Partition, Vector Selection Array=$System.Offset_Front)
     False Vector=&SelectVectorFromArray(Direct Array Index=$Particles.PartitionParticles.Partition, Vector Selection Array=$System.Offset_Car)
    )
   )
   &InitialMeshOrientation(Mesh Orientation Mode=::NewEnumerator2, Use Rotation Vector=true)
   /SetVariables(
    Particles.CameraOffset_FakeLit=10
    Particles.Color_FakeLit:LinearColor=&MultiplyLinearColor(Linear Color=<rgba 0 5 3.75 1>, Scale Factor=<1 1 1 1>)
    Particles.Position_FakeLit:NiagaraPosition=&AddVectorToPosition(
     Position=$Particles.Position
     Vector=&TransformVector(
      Vector=&SelectVectorFromArray(
       Direct Array Index=$Particles.PartitionParticles.Partition
       Vector Selection Array=DataInterface() {propertyValues {FloatData [<0 50 15>, <0 -50 15>, <0 50 15>, <0 -50 15>]}}
      )
      Destination Space=::Local
     )
    )
   ) {isSetParameters=true}
   /SetVariables(Particles.PositionSkid:NiagaraPosition=&AddVectorToPosition(Position=$Particles.Position, Vector=<-5 0 0>)) {isSetParameters=true}
  ]
  particleUpdateScript [
   &ParticleState()
   &ScaleColor(
    Scale Alpha=&FloatFromCurve(FloatCurve=DataInterface() {propertyValues {Curve {keys [
         {arriveTangent=-0.2241, interpMode=RCIM_Cubic, leaveTangent=-0.2241, tangentMode=RCTM_User, time=0, value=1}
         {arriveTangent=-3.1123, interpMode=RCIM_Cubic, leaveTangent=-3.1123, tangentMode=RCTM_User, time=1, value=0}
        ]}}})
    Scale RGB=&MakeVectorFromLinearColorRGB(Color=&ColorFromCurve(ColorCurve=DataInterface() {propertyValues {
        AlphaCurve {keys [<at 0 1 RCIM_Cubic>]}
        BlueCurve {keys [<at 0.0137 2.0833>,<at 0.1369 0.2517>,<at 0.9732 0.34>]}
        GreenCurve {keys [<at 0.0137 2.5>,<at 0.1369 0.5>,<at 0.9732 0.5>]}
        RedCurve {keys [<at 0.0137 0>,<at 0.1369 0.0045>,<at 0.9732 0>]}
       }}))
    ScaleA=true
    ScaleRGB=true
   )
   /SetVariables(Particles.Color_FakeLit:LinearColor=&MultiplyLinearColor(
     Linear Color=$Particles.Initial.Color_FakeLit
     Scale Factor=&MakeFloat4FromVectorAndFloat(
      "Vector (RGB)"=<1 1 1>
      "Float (Alpha)"=&FloatFromCurve(
       FloatCurve=DataInterface() {propertyValues {Curve {keys [
           {arriveTangent=13.9911, interpMode=RCIM_Cubic, leaveTangent=13.9911, tangentMode=RCTM_User, time=0, value=0}
           {arriveTangent=1, interpMode=RCIM_Cubic, leaveTangent=1, time=0.2, value=1}
           <at 1 1>
          ]}}}
       CurveIndex=$System.NormalizedVelocityRange
       Scale Curve=0.88
      )
     )
    )) {isSetParameters=true}
   /SetVariables(Particles.DMP_FakeLight_01:Vector4f=&MakeVector4(X=0.5, Y=1, Z=1, W=0)) {isSetParameters=true}
   &DynamicMaterialParameters(
    Index 0 Param 1=&FloatFromCurve(FloatCurve=DataInterface() {propertyValues {Curve {keys [<at 0.1 0.1 RCIM_Cubic>,<at 0.8 1>]}}})
    Index 0 Param 2=2
    Index 0 Param 4=&FloatFromCurve(FloatCurve=DataInterface() {propertyValues {Curve {keys [<at 0 1>,<at 1 3>]}}})
    Module.Param0WriteEnabled=true
    Module.Param1WriteEnabled=true
    Module.Param2WriteEnabled=false
    Module.Param3WriteEnabled=true
   )
   &ScaleMeshSize(Scale Factor=&Multiply_VectorByFloat(
     Vector=&MakeVector(
      X=1
      Y=1
      Z=&FloatFromCurve(
       FloatCurve=DataInterface() {propertyValues {Curve {keys [
           {arriveTangent=2.4702, interpMode=RCIM_Cubic, leaveTangent=2.4702, tangentMode=RCTM_User, time=0, value=0}
           {arriveTangent=0.9995, interpMode=RCIM_Cubic, leaveTangent=0.9995, tangentMode=RCTM_User, time=0.4, value=0.75}
           {arriveTangent=-0.0405, interpMode=RCIM_Cubic, leaveTangent=-0.0405, tangentMode=RCTM_User, time=1, value=1}
          ]}}}
       CurveIndex=$System.NormalizedVelocityRange
      )
     )
     Float=&FloatFromCurve(
      FloatCurve=DataInterface() {propertyValues {Curve {keys [
          {arriveTangent=2.4702, interpMode=RCIM_Cubic, leaveTangent=2.4702, tangentMode=RCTM_User, time=0, value=0}
          {arriveTangent=0.9995, interpMode=RCIM_Cubic, leaveTangent=0.9995, tangentMode=RCTM_User, time=0.4, value=0.75}
          {arriveTangent=-0.0405, interpMode=RCIM_Cubic, leaveTangent=-0.0405, tangentMode=RCTM_User, time=1, value=1}
         ]}}}
      Scale Curve=2
     )
    )) {module=ScaleMeshSize001}
   &SolveForcesAndVelocity(Clamp Velocity=false, Limit Acceleration=false)
  ]
  rendererClasses [&NiagaraMeshRendererProperties, &NiagaraSpriteRendererProperties]
  renderers [
   {rendererClass=&NiagaraMeshRendererProperties, rendererIndex=0}
   {rendererClass=&NiagaraSpriteRendererProperties, rendererIndex=1}
  ]
 }
 Smoke_Wheel {
  emitterUpdateScript [
   &EmitterState(
    Inactive Response=::NewEnumerator1
    Life Cycle Mode=::NewEnumerator1
    Loop Behavior=::NewEnumerator1
    Loop Delay=$System.Delay
    Loop Duration Mode=::NewEnumerator1
    UseLoopDelay=true
   )
   &SpawnBurst_Instantaneous(
    Spawn Count=&MakeCustomIntFromBool(Bool=$User.isFrontPart, True INT=6, False INT=12)
    Use Spawn Probability=false
    UseLoopCountLimit=false
   )
  ]
  particleSpawnScript [
   &PartitionParticles(Number Of Partitions=&MakeCustomIntFromBool(Bool=$User.isFrontPart, True INT=6, False INT=12))
   /SetVariables() {isSetParameters=true}
   &InitializeParticle(
    Color=<rgba 0.01 0.01 0.01 0.5>
    Mesh Scale=&RandomRangeVector(Minimum=<1 1 1.5>, Maximum=<1 1 1.75>)
    Mesh Scale Mode=::NewEnumerator4
    UsePositionOffset=false
    Write Material Random=true
   )
   &InitialMeshOrientation(
    Mesh Axis To Orient=<0 0 1>
    Mesh Orientation Mode=::NewEnumerator2
    Orientation Vector=<-1 0 0>
    Rotation=&MakeCustomVectorFromBool(
     Bool=$User.isFrontPart
     True Vector=&SelectVectorFromArray(
      Direct Array Index=&ModuloInt(A=$Particles.PartitionParticles.Partition, B=2)
      Vector Selection Array=DataInterface() {propertyValues {FloatData [<0 0.025 0>, <0.025 0 0.5>]}}
     )
     False Vector=&SelectVectorFromArray(
      Direct Array Index=&ModuloInt(A=$Particles.PartitionParticles.Partition, B=4)
      Vector Selection Array=DataInterface() {propertyValues {FloatData [<0 0.025 0>, <0.025 0 0.5>, <0 0.025 0>, <0.025 0 0.5>]}}
     )
    )
    Use Rotation Vector=true
   )
   &CameraOffset(Camera Offset Amount=50)
   &SystemLocation(
    MaskSpawnGroup=false
    Offset=&MakeCustomVectorFromBool(
     Bool=$User.isFrontPart
     True Vector=&SelectVectorFromArray(
      Direct Array Index=&ModuloInt(A=$Particles.PartitionParticles.Partition, B=2)
      Vector Selection Array=$System.Offset_Front
     )
     False Vector=&SelectVectorFromArray(
      Direct Array Index=&ModuloInt(A=$Particles.PartitionParticles.Partition, B=4)
      Vector Selection Array=$System.Offset_Car
     )
    )
   )
  ]
  particleUpdateScript [
   &DynamicMaterialParameters(
    Index 0 Param 1=&SelectFloatFromArray(
     Direct Array Index=$Particles.PartitionParticles.Partition
     Float Selection Array=DataInterface() {propertyValues {FloatData [150,150,200,150,150,200,150,150,200,150,150,200]}}
    )
    Index 0 Param 2=&SelectFloatFromArray(
     Direct Array Index=$Particles.PartitionParticles.Partition
     Float Selection Array=DataInterface() {propertyValues {FloatData [60,70,50,60,70,50,60,70,50,60,70,50]}}
    )
    Index 0 Param 3=&SelectFloatFromArray(
     Direct Array Index=$Particles.PartitionParticles.Partition
     Float Selection Array=DataInterface() {propertyValues {FloatData [1,1,1.5,1,1,1.5,1,1,1.5,1,1,1.5]}}
    )
    Index 0 Param 4=&SelectFloatFromArray(
     Direct Array Index=$Particles.PartitionParticles.Partition
     Float Selection Array=DataInterface() {propertyValues {FloatData [0,0.4,0,0,0.3,0,0,0.4,0,0,0.3,0]}}
    )
    Index 1 Param 1=1
    Index 1 Param 2=&FloatFromCurve(FloatCurve=DataInterface() {propertyValues {Curve {keys [{tangentMode=RCTM_User, time=0, value=0}, {tangentMode=RCTM_User, time=0.5, value=1}]}}})
    Module.Index0Param0Write=true
    Module.Index0Param1Write=true
    Module.Index0Param2Write=true
    Module.Index0Param3Write=true
    Module.Index1Param0Write=true
    Module.Index1Param1Write=true
    Module.Index1Param2Write=true
    Module.Index1Param3Write=true
    Write Parameter Index 1=true
   )
   &ParticleState(Kill Particles When Lifetime Has Elapsed=false, Let Infinitely Lived Particles Die When Emitter Deactivates=true)
   &ScaleMeshSize(Scale Factor=&MakeVector(
     X=0.5
     Y=1
     Z=&Multiply_Float(
      A=1.5
      B=&FloatFromCurve(
       FloatCurve=DataInterface() {propertyValues {Curve {keys [
           {arriveTangent=1.7693, interpMode=RCIM_Cubic, leaveTangent=1.7693, tangentMode=RCTM_User, time=0, value=0}
           {interpMode=RCIM_Cubic, tangentMode=RCTM_User, time=1, value=1}
          ]}}}
       CurveIndex=$System.NormalizedVelocityRange
       Scale Curve=2
      )
     )
    ))
   /SetVariables(
    Particles.EnergyTrailColor:LinearColor=&RandomRangeLinearColor(Minimum=<rgba 0 1 0.8333 0.1>, Maximum=<rgba 0 1 0.7683 0.1>)
    Particles.EnergyTrailColorSecondary:LinearColor=<rgba 0 4 3.3333 2.5>
    Particles.EnergyTrailDMP:Vector4f=&MakeVector4(X=50, Y=10, Z=0.1, W=0.5)
    Particles.EnergyTrailDMPSecondary:Vector4f=&MakeVector4(
     X=&RandomRangeFloat(Minimum=150, Maximum=200)
     Y=&RandomRangeFloat(Minimum=50, Maximum=70)
     Z=&RandomRangeFloat(Minimum=0.75, Maximum=1.5)
     W=&RandomRangeFloat(Minimum=0.75, Maximum=1.5)
    )
    Particles.EnergyTrailSize:Vector3f=&Multiply_Vector(A=$Particles.Scale, B=<0.3 0.3 0.25>)
    Particles.EnergyTrailSizeSecondary:Vector3f=&Multiply_Vector(A=$Particles.Scale, B=<0.5 0.5 0.75>)
   ) {isSetParameters=true}
  ]
  rendererClasses [&NiagaraMeshRendererProperties]
  renderers [
   {rendererClass=&NiagaraMeshRendererProperties, rendererIndex=0}
   {rendererClass=&NiagaraMeshRendererProperties, rendererIndex=1}
   {rendererClass=&NiagaraMeshRendererProperties, rendererIndex=2}
  ]
 }
 Spark {
  simTarget=GPUComputeSim
  emitterUpdateScript [
   &EmitterState(Life Cycle Mode=::NewEnumerator1, Loop Duration=1, UseLoopDelay=false)
   &SpawnRate(
    Spawn Probability=&FloatFromCurve(
     FloatCurve=DataInterface() {propertyValues {Curve {keys [
         {arriveTangent=4.1785, interpMode=RCIM_Cubic, leaveTangent=4.1785, tangentMode=RCTM_User, time=0, value=0}
         {arriveTangent=1.1179, interpMode=RCIM_Cubic, leaveTangent=1.1179, tangentMode=RCTM_User, time=0.2042, value=0.6228}
         {arriveTangent=0.0224, interpMode=RCIM_Cubic, leaveTangent=0.0224, tangentMode=RCTM_User, time=1, value=1}
        ]}}}
     CurveIndex=$System.NormalizedVelocityRange
    )
    SpawnRate=&MakeCustomFloatFromBool(Bool=$User.isFrontPart, True Float=100, False Float=200)
    Use Spawn Probability=true
   )
  ]
  particleSpawnScript [
   &PartitionParticles(Number Of Partitions=&MakeCustomIntFromBool(Bool=$User.isFrontPart, True INT=2, False INT=4))
   &InitializeParticle(
    Color=<rgba 0.01 0.25 0.21 1>
    Lifetime Min=0.4
    Lifetime Mode=::NewEnumerator1
    Sprite Size Max=<2 4>
    Sprite Size Min=<1 2>
    Sprite Size Mode=::NewEnumerator2
    UsePositionOffset=false
    Write Material Random=true
   )
   &SphereLocation(
    MaskSpawnGroup=false
    Offset=&MakeCustomVectorFromBool(
     Bool=$User.isFrontPart
     True Vector=&SelectVectorFromArray(Direct Array Index=$Particles.PartitionParticles.Partition, Vector Selection Array=$System.Offset_Front)
     False Vector=&SelectVectorFromArray(Direct Array Index=$Particles.PartitionParticles.Partition, Vector Selection Array=$System.Offset_Car)
    )
    Override Seed=false
    Sphere Radius=&RandomRangeFloat(Minimum=10, Maximum=20)
    Surface Only=false
    UseNonUniformScale=false
    UseOffset=true
    UseSphereOrientation=false
   )
   &AddVelocity(
    Velocity=&SelectVectorFromArray(
     Direct Array Index=$Particles.PartitionParticles.Partition
     Direct Array Mode=::NewEnumerator1
     Vector Selection Array=DataInterface() {propertyValues {FloatData [<-1 0.25 0>, <-1 -0.25 0>]}}
    )
    Velocity Speed Scale=&RandomRangeFloat(Minimum=750, Maximum=1250)
   )
   &CameraOffset(Camera Offset Amount=50)
  ]
  particleUpdateScript [
   &CurlNoiseForce(
    Noise Strength=&Multiply_Float(
     A=&RandomRangeFloat(Minimum=2500, Maximum=5000)
     B=&FloatFromCurve(
      FloatCurve=DataInterface() {propertyValues {Curve {keys [{tangentMode=RCTM_User, time=0, value=0}, {tangentMode=RCTM_User, time=1, value=1}]}}}
      CurveIndex=&ReturnNormalizedExecIndex()
     )
    )
    Pan Noise=true
    Pan Noise Field=<0.5 0.5 0.5>
   )
   &AccelerationForce(Acceleration=&Multiply_Vector(
     A=<0 0 0.1>
     B=&VectorFromFloat(Value=&RandomRangeFloat(Minimum=1500, Maximum=2500))
    ))
   &Drag(
    Drag=&FloatFromCurve(
     FloatCurve=DataInterface() {propertyValues {Curve {keys [
         {interpMode=RCIM_Cubic, tangentMode=RCTM_User, time=0, value=0}
         {arriveTangent=3.5282, interpMode=RCIM_Cubic, leaveTangent=3.5282, tangentMode=RCTM_User, time=1, value=1}
        ]}}}
     CurveIndex=&ReturnNormalizedExecIndex()
     Scale Curve=5
    )
    Use Linear Drag=true
    Use Rotational Drag=false
   )
   &SolveForcesAndVelocity(Clamp Velocity=false, Limit Acceleration=false)
   &ParticleState()
   &ScaleSpriteSize(Scale Factor=&Vector2DFromFloat(Value=&FloatFromCurve(FloatCurve=DataInterface() {propertyValues {Curve {keys [<at 0 0.2>,<at 0.2 1>,<at 1 0>]}}}, Scale Curve=2)))
   &ScaleSpriteSizeBySpeed(Max Scale Factor=<0.05 5>, Min Scale Factor=<1 0.5>)
   &ScaleColor(
    Scale Alpha=&FloatFromCurve(
     FloatCurve=DataInterface() {propertyValues {Curve {keys [<at 0 0>,<at 0.2 1>]}}}
     Scale Curve=&FNTimeOfDayScalar(Night Scale=0.6, Morning Scale=0.4, Evening Scale=0.4)
    )
    Scale RGB=&VectorFromFloat(Value=&Multiply_Float(
      A=&FloatFromCurve(FloatCurve=DataInterface() {propertyValues {Curve {keys [<at 0 50>,<at 1 10>]}}})
      B=&FloatFromCurve(FloatCurve=DataInterface() {propertyValues {Curve {keys [
           <at 0 1 RCIM_Cubic>
           {arriveTangent=-0.2561, interpMode=RCIM_Cubic, leaveTangent=-0.2561, time=0.2, value=0.15}
           {arriveTangent=-0.2649, interpMode=RCIM_Cubic, leaveTangent=-0.2649, time=0.3695, value=2}
           {arriveTangent=0.0722, interpMode=RCIM_Cubic, leaveTangent=0.0722, time=0.5434, value=0.059}
           {arriveTangent=0.1149, interpMode=RCIM_Cubic, leaveTangent=0.1149, time=0.7139, value=1.5}
           {arriveTangent=0.2439, interpMode=RCIM_Cubic, leaveTangent=0.2439, time=0.9, value=0.1}
           <at 1 1 RCIM_Cubic>
          ]}}})
     ))
    ScaleA=true
    ScaleRGB=true
   ) {module=ScaleColor001}
   &KillParticlesInVolume(Box Size:Vector3f=<1950 450 150>, Kill Shape=::NewEnumerator1, Origin Offset:Vector3f=<0 0 150>)
  ]
  rendererClasses [&NiagaraSpriteRendererProperties]
  renderers [{rendererClass=&NiagaraSpriteRendererProperties}]
 }
}
summary {
 systemName=NS_FNM_Train_Wheels_Car
 emitters [
  {emitterName=Spark, simTarget=GPUComputeSim, rendererClasses [&NiagaraSpriteRendererProperties]}
  {emitterName=Smoke_Wheel, rendererClasses [&NiagaraMeshRendererProperties]}
  {emitterName=SkidSharp, rendererClasses [&NiagaraMeshRendererProperties, &NiagaraSpriteRendererProperties]}
  {bEnabled=false, emitterName=FN_DebugGnomon, rendererClasses [&NiagaraMeshRendererProperties]}
 ]
 userVariables [
  User.bIsFrontEnd {type=NiagaraBool}
  User.bIsFrontEndPreview {type=NiagaraBool}
  User.isFrontPart {type=NiagaraBool}
 ]
}
